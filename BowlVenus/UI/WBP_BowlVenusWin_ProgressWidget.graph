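# HUD widget that shows capture point progress as a radial or diamond bar, chosen by the capture point's gameplay tag.
# Construct creates dynamic materials for both bars and starts the widget collapsed.
# Progress, contested, reset and claim-outcome functions set fill, team color and status icon, writing to both materials.
# OnClampStateChanged shows the matching edge arrow when the widget is clamped to the screen.
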WidgetBlueprint /BowlVenus/UI/WBP_BowlVenusWin_ProgressWidget
graphs [
 EventGraph {nodes [
   Event Construct()
   > Setup Progress Materials()
   > SetVisibility(self=Get Overlay_ClampArrows() -> (Overlay_ClampArrows:Overlay Object Reference), InVisibility=Collapsed)
   > SetVisibility(InVisibility=Collapsed)
  ]}
 On Progress Contested {nodes [
   n1 = On Progress Contested!() -> (then, IsContested:bool=~) {is=title K2Node_FunctionEntry}
   > n2 = Set IsContested(IsContested:bool=@n1) -> (Output_Get:bool)
   > Branch(Condition=@n2)
   > Set Progress Color!(execute, self, Value:Linear Color Structure=Get ContestedColor() -> (ContestedColor:Linear Color Structure)) -> (then) {is=title K2Node_CallFunction}
   > SetBrushFromTexture!(
    execute
    self:Image Object Reference=Get Image_StatusIcon() -> (Image_StatusIcon:Image Object Reference)
    Texture:Texture 2D Object Reference=Get ContestedIcon() -> (ContestedIcon:Texture 2D Object Reference)
    bMatchSize:bool=false
   ) -> (then)
  ]}
 OnClampStateChanged {nodes [
   OnClampStateChanged!() -> (then, ClampToScreenDirection:EClampToScreenDirection Enum=Left) {is=title K2Node_FunctionEntry}
   > n1 = Branch(Condition="Not Equal (Enum)"!(
     A:EClampToScreenDirection Enum=Get ClampToScreenDirection() -> (ClampToScreenDirection:EClampToScreenDirection Enum=Left)
     B:EClampToScreenDirection Enum=None
    ) -> (ReturnValue:bool))
   > SetVisibility(self=Get Overlay_ClampArrows(), InVisibility=HitTestInvisible)
   n2 = Get ClampToScreenDirection() -> (ClampToScreenDirection=Left)
   > SetVisibility(
    self=Get Image_ClampLeft() -> (Image_ClampLeft:Image Object Reference)
    InVisibility=Select(Option 0=Collapsed, Option 1=HitTestInvisible, Index="Equal (Enum)"(A:EClampToScreenDirection Enum=@n2, B:EClampToScreenDirection Enum=Left))
   )
   n3 = Get ClampToScreenDirection() -> (ClampToScreenDirection=Left)
   > SetVisibility(
    self=Get Image_ClampRight() -> (Image_ClampRight:Image Object Reference)
    InVisibility=Select(Option 0=Collapsed, Option 1=HitTestInvisible, Index="Equal (Enum)"(A=@n3, B=Right))
   )
   n4 = Get ClampToScreenDirection() -> (ClampToScreenDirection=Left)
   > SetVisibility(
    self=Get Image_ClampTop() -> (Image_ClampTop:Image Object Reference)
    InVisibility=Select(Option 0=Collapsed, Option 1=HitTestInvisible, Index="Equal (Enum)"(A=@n4, B=Top))
   )
   n5 = Get ClampToScreenDirection() -> (ClampToScreenDirection=Left)
   > SetVisibility(
    self=Get Image_ClampBottom() -> (Image_ClampBottom:Image Object Reference)
    InVisibility=Select(Option 0=Collapsed, Option 1=HitTestInvisible, Index="Equal (Enum)"(A=@n5, B=Bottom))
   )
   SetVisibility(execute=@n1.else, self=Get Overlay_ClampArrows(), InVisibility=Collapsed)
  ]}
 ResetProgress {nodes [
   n1 = ResetProgress!() -> (then, ResetProgressTo:double=~) {is=title K2Node_FunctionEntry}
   > Update Progress Value!(execute, self, Parameter Name:Name=Progress, Value:double=@n1) -> (then)
   > SetBrushFromTexture(self=Get Image_StatusIcon(), Texture=Get FriendlyProgress() -> (FriendlyProgress:Texture 2D Object Reference), bMatchSize=false)
   > Set Progress Color(Value=Get FriendlyColor() -> (FriendlyColor:Linear Color Structure))
  ]}
 Set Progress Color {nodes [
   n1 = Set Progress Color!() -> (then, Value:Linear Color Structure) {is=title K2Node_FunctionEntry}
   > SetVectorParameterValue(self=Get RadialDynamicMaterial() -> (RadialDynamicMaterial:Material Instance Dynamic Object Reference), ParameterName=Shape Color, Value=@n1)
   > SetVectorParameterValue(
    self=Get DiamondDynamicMaterial() -> (DiamondDynamicMaterial:Material Instance Dynamic Object Reference)
    ParameterName=FillColor
    Value=Get Value() -> (Value:Linear Color Structure)
   )
  ]}
 Setup Progress Materials {nodes [
   Setup Progress Materials() {is=title K2Node_FunctionEntry}
   > n1 = GetDynamicMaterial(self=Get Image_Radial_ProgressBar() -> (Image_Radial_ProgressBar:Image Object Reference))
   > n2 = Set RadialDynamicMaterial(RadialDynamicMaterial:Material Instance Dynamic Object Reference=@n1) -> (Output_Get:Material Instance Dynamic Object Reference)
   > SetScalarParameterValue(self=@n2, ParameterName=Circle Stroke, Value=0.1)
   > SetVectorParameterValue(self=@n2, ParameterName=Bg Color, Value=(R=0.0953,G=0.0953,B=0.0953,A=1))
   > n3 = GetDynamicMaterial(self=Get Image_Diamond_ProgressBar() -> (Image_Diamond_ProgressBar:Image Object Reference))
   > n4 = Set DiamondDynamicMaterial(DiamondDynamicMaterial:Material Instance Dynamic Object Reference=@n3) -> (Output_Get:Material Instance Dynamic Object Reference)
   > SetVectorParameterValue(self=@n4, ParameterName=BGColor, Value=(R=0.0953,G=0.0953,B=0.0953,A=1))
   > Update Progress Value(Parameter Name=Progress, Value=0)
  ]}
 SetupProgressBarToShow {nodes [
   n1 = SetupProgressBarToShow!() -> (then, CapturePointActor:Actor Object Reference) {is=title K2Node_FunctionEntry}
   > Is Valid(InputObject=@n1)
   > n2 = Cast To BGA_POIBanner_CapturePoint!(execute, Object=@n1) -> (then, CastFailed, AsBGA POIBanner Capture Point:BGA POIBanner Capture Point Object Reference) {is=title K2Node_DynamicCast}
   n3 = Get Owned Gameplay Tags!(self:Gameplay Tag Asset Interface Interface=@n2) -> (ReturnValue:Gameplay Tag Container Structure) {is=title K2Node_CallFunction}
   > SetActiveWidget(
    self=Get Switcher_ProgressBar_Shape() -> (Switcher_ProgressBar_Shape:Common Visibility Switcher Object Reference)
    Widget=Select(
     Option 0:Widget Object Reference=Get Overlay_Diamond() -> (Overlay_Diamond:Overlay Object Reference)
     Option 1:Widget Object Reference=Get Overlay_Radial() -> (Overlay_Radial:Overlay Object Reference)
     Index=HasTag!(
      TagContainer:"Gameplay Tag Container Structure (by ref)"=@n3
      Tag:Gameplay Tag Structure=(TagName="POIBanner.HUD.ProgressBar.Round")
      bExactMatch:bool=false
     ) -> (ReturnValue:bool)
    ) -> (ReturnValue:Widget Object Reference)
   )
  ]}
 ShowClaimOutcome {nodes [
   n1 = ShowClaimOutcome!() -> (then, Won:bool=~) {is=title K2Node_FunctionEntry}
   > SetBrushFromTexture(
    self=Get Image_StatusIcon()
    Texture=Select(
     Option 0:Texture 2D Object Reference=Get EnemyClaimedIcon() -> (EnemyClaimedIcon:Texture 2D Object Reference)
     Option 1:Texture 2D Object Reference=Get FriendlyClaimedIcon() -> (FriendlyClaimedIcon:Texture 2D Object Reference)
     Index=@n1
    ) -> (ReturnValue:Texture 2D Object Reference)
    bMatchSize=false
   )
   > Set Progress Color(Value=SelectColor(
     A=Get FriendlyClaimColor() -> (FriendlyClaimColor:Linear Color Structure)
     B=Get EnemyColor() -> (EnemyColor:Linear Color Structure)
     bPickA=Get Won() -> (Won:bool)
    )) {is=title K2Node_CallFunction}
  ]}
 Update Claim Progress {nodes [
   n1 = Update Claim Progress!() -> (then, IsFriendlyTeam:bool=~, CurrentProgress:double=~) {is=title K2Node_FunctionEntry}
   > Update Progress Value(Parameter Name=Progress, Value=@n1.CurrentProgress)
   > Set Progress Color(Value=SelectColor(A=Get FriendlyColor(), B=Get EnemyColor(), bPickA=Get IsFriendlyTeam() -> (IsFriendlyTeam:bool)))
   > SetBrushFromTexture(
    self=Get Image_StatusIcon()
    Texture=Select(Option 0=Get EnemyProgress() -> (EnemyProgress:Texture 2D Object Reference), Option 1=Get FriendlyProgress(), Index=Get IsFriendlyTeam())
    bMatchSize=false
   )
  ]}
 Update Progress Value {nodes [
   n1 = Update Progress Value!() -> (then, Parameter Name:Name, Value:double=~) {is=title K2Node_FunctionEntry}
   > SetScalarParameterValue(self=Get DiamondDynamicMaterial(), ParameterName=@n1, Value=@n1.Value)
   > SetScalarParameterValue(self=Get RadialDynamicMaterial(), ParameterName=Get Parameter Name() -> (Parameter Name:Name=None), Value=Get Value() -> (Value:double))
  ]}
 Update Widget Location Type {nodes [
   n1 = Update Widget Location Type!() -> (then, InIsStaticWidget:bool=~) {is=title K2Node_FunctionEntry}
   > Set IsStaticWidget(IsStaticWidget:bool=@n1) -> (Output_Get:bool)
  ]}
]
variables [
 ContestedColor {type=LinearColor, default=<rgba 0.7529 0 0 1>}
 ContestedIcon {type=Texture2D, default=/BowlVenus/UI/Textures/T_UI_BowlVenus_Icon_Minimap_Contested}
 DiamondDynamicMaterial {type=MaterialInstanceDynamic}
 EnemyClaimedIcon {type=Texture2D, default=/BowlVenus/UI/Textures/T_UI_BowlVenus_Icon_Minimap_EnemyClaimed}
 EnemyColor {type=LinearColor, default=<rgba 0.8308 0.1651 0 1>}
 EnemyProgress {type=Texture2D, default=/BowlVenus/UI/Textures/T_UI_BowlVenus_Icon_Minimap_EnemyProgress}
 FriendlyClaimColor {type=LinearColor, default=<rgba 0 0.7231 0.0021 1>}
 FriendlyClaimedIcon {type=Texture2D, default=/BowlVenus/UI/Textures/T_UI_BowlVenus_Icon_Minimap_Claimed}
 FriendlyColor {type=LinearColor, default=<rgba 0.0273 0.5906 0.8632 1>}
 FriendlyProgress {type=Texture2D, default=/BowlVenus/UI/Textures/T_UI_BowlVenus_Icon_Minimap_Uncontested}
 Is Friendly Team {type=boolean}
 IsContested {type=boolean}
 IsStaticWidget {type=boolean, default=true}
 RadialDynamicMaterial {type=MaterialInstanceDynamic}
]
widgets {
 info {inheritedWidgetCount=1, parentClass=&FortHUDElementWidget, rootWidgetClass=&Overlay, widgetCount=19}
 widgets [
  Overlay_Root {widgetClassPath=&Overlay, widget=Overlay_Root}
  Slot_BehindProgress {
   widgetClassPath=&NamedSlot
   bIsVariable=true
   parent=Overlay_Root
   widget=Slot_BehindProgress
   slot {class=OverlaySlot, properties {horizontalAlignment=HAlign_Fill, verticalAlignment=VAlign_Fill}}
  }
  Overlay_Main {
   widgetClassPath=&Overlay
   parent=Overlay_Root
   widget=Overlay_Main
   slot {class=OverlaySlot, properties {horizontalAlignment=HAlign_Center, verticalAlignment=VAlign_Center}}
  }
  SizeBoxContainer {
   widgetClassPath=&SizeBox
   parent=Overlay_Main
   widget=SizeBoxContainer
   properties {bOverride_HeightOverride=true, bOverride_WidthOverride=true, heightOverride=100, maxDesiredWidth=64, minDesiredWidth=64, widthOverride=100}
   slot {class=OverlaySlot, properties {horizontalAlignment=HAlign_Center, verticalAlignment=VAlign_Center}}
  }
  Overlay_ProgressRadial {widgetClassPath=&Overlay, parent=SizeBoxContainer, widget=Overlay_ProgressRadial, slot {class=SizeBoxSlot}}
  Switcher_ProgressBar_Shape {
   widgetClassPath=&CommonVisibilitySwitcher
   bIsVariable=true
   parent=Overlay_ProgressRadial
   widget=Switcher_ProgressBar_Shape
   slot {class=OverlaySlot, properties {horizontalAlignment=HAlign_Fill, verticalAlignment=VAlign_Fill}}
  }
  Overlay_Radial {
   widgetClassPath=&Overlay
   bIsVariable=true
   parent=Switcher_ProgressBar_Shape
   widget=Overlay_Radial
   slot {class=CommonVisibilitySwitcherSlot}
  }
  Image_Radial_BGOpaque {
   widgetClassPath=&Image
   parent=Overlay_Radial
   widget=Image_Radial_BGOpaque
   properties {brush {
     bIsDynamicallyLoaded=false
     drawAs=Image
     imageType=NoImage
     mirroring=NoMirror
     resourceName=None
     resourceObject=/Game/Athena/UI/Materials/M_UI_Circle
     tiling=NoTile
     imageSize=<32 32>
     margin=<ltrb 0 0 0 0>
     outlineSettings {
      bUseBrushTransparency=false
      roundingType=HalfHeightRadius
      width=0
      color {colorUseRule=UseColor_Specified, specifiedColor=<rgba 0 0 0 0>}
      cornerRadii=<0 0 0 0>
     }
     tintColor {colorUseRule=UseColor_Specified, specifiedColor=<rgba 0 0 0 0.5>}
     uVRegion {bIsValid=false, max=<0 0>, min=<0 0>}
    }}
   slot {class=OverlaySlot, properties {horizontalAlignment=HAlign_Fill, verticalAlignment=VAlign_Fill, padding=<ltrb 5 5 5 5>}}
  }
  Image_Radial_ProgressBar {
   widgetClassPath=&Image
   bIsVariable=true
   parent=Overlay_Radial
   widget=Image_Radial_ProgressBar
   properties {brush {
     bIsDynamicallyLoaded=false
     drawAs=Image
     imageType=NoImage
     mirroring=NoMirror
     resourceName=None
     resourceObject=/Game/UI/Foundation/Toasts/Materials/MI_UI_RadialProgress
     tiling=NoTile
     imageSize=<68 68>
     margin=<ltrb 0 0 0 0>
     outlineSettings {
      bUseBrushTransparency=false
      roundingType=HalfHeightRadius
      width=0
      color {colorUseRule=UseColor_Specified, specifiedColor=<rgba 0 0 0 0>}
      cornerRadii=<0 0 0 0>
     }
     tintColor {colorUseRule=UseColor_Specified, specifiedColor=<rgba 1 1 1 1>}
     uVRegion {bIsValid=false, max=<0 0>, min=<0 0>}
    }}
   slot {class=OverlaySlot, properties {horizontalAlignment=HAlign_Fill, verticalAlignment=VAlign_Fill, padding=<ltrb 5 5 5 5>}}
  }
  Overlay_Diamond {
   widgetClassPath=&Overlay
   bIsVariable=true
   parent=Switcher_ProgressBar_Shape
   widget=Overlay_Diamond
   slot {class=CommonVisibilitySwitcherSlot, properties {padding=<ltrb -10 -10 -10 -10>}}
  }
  Image_Diamond_BGOpaque {
   widgetClassPath=&Image
   parent=Overlay_Diamond
   widget=Image_Diamond_BGOpaque
   properties {brush {
     bIsDynamicallyLoaded=false
     drawAs=Image
     imageType=NoImage
     mirroring=NoMirror
     resourceName=None
     resourceObject=/Game/Athena/HUD/Compass/M_UI_DiamondMarker
     tiling=NoTile
     imageSize=<32 32>
     margin=<ltrb 0 0 0 0>
     outlineSettings {
      bUseBrushTransparency=false
      roundingType=HalfHeightRadius
      width=0
      color {colorUseRule=UseColor_Specified, specifiedColor=<rgba 0 0 0 0>}
      cornerRadii=<0 0 0 0>
     }
     tintColor {colorUseRule=UseColor_Specified, specifiedColor=<rgba 0 0 0 0.5>}
     uVRegion {bIsValid=false, max=<0 0>, min=<0 0>}
    }}
   slot {class=OverlaySlot, properties {horizontalAlignment=HAlign_Fill, verticalAlignment=VAlign_Fill}}
  }
  Image_Diamond_ProgressBar {
   widgetClassPath=&Image
   bIsVariable=true
   parent=Overlay_Diamond
   widget=Image_Diamond_ProgressBar
   properties {
    visibility=SelfHitTestInvisible
    brush {
     bIsDynamicallyLoaded=false
     drawAs=Image
     imageType=NoImage
     mirroring=NoMirror
     resourceName=None
     resourceObject=/Game/Athena/HUD/Materials/M_UI_DiamondProgress_RadialFill_Inst
     tiling=NoTile
     imageSize=<68 68>
     margin=<ltrb 0 0 0 0>
     outlineSettings {
      bUseBrushTransparency=false
      roundingType=HalfHeightRadius
      width=0
      color {colorUseRule=UseColor_Specified, specifiedColor=<rgba 0 0 0 0>}
      cornerRadii=<0 0 0 0>
     }
     tintColor {colorUseRule=UseColor_Specified, specifiedColor=<rgba 1 1 1 1>}
     uVRegion {bIsValid=false, max=<0 0>, min=<0 0>}
    }
    renderTransform {angle=-45, scale=<1 1>, shear=<0 0>, translation=<0 0>}
   }
   slot {class=OverlaySlot, properties {horizontalAlignment=HAlign_Center, verticalAlignment=VAlign_Center}}
  }
  Image_StatusIcon {
   widgetClassPath=&Image
   bIsVariable=true
   parent=Overlay_ProgressRadial
   widget=Image_StatusIcon
   properties {
    brush {
     bIsDynamicallyLoaded=false
     drawAs=Image
     imageType=NoImage
     mirroring=NoMirror
     resourceName=None
     resourceObject=/Game/Athena/UI/InGame/Creative/PropertyEditor/Icon_Picker/Textures/64x64/T_UI_IconLibrary_flag_64
     tiling=NoTile
     imageSize=<48 48>
     margin=<ltrb 0 0 0 0>
     outlineSettings {
      bUseBrushTransparency=false
      roundingType=HalfHeightRadius
      width=0
      color {colorUseRule=UseColor_Specified, specifiedColor=<rgba 0 0 0 0>}
      cornerRadii=<0 0 0 0>
     }
     tintColor {colorUseRule=UseColor_Specified, specifiedColor=<rgba 1 1 1 1>}
     uVRegion {bIsValid=false, max=<0 0>, min=<0 0>}
    }
    renderTransformPivot=<1 0.5>
   }
   slot {class=OverlaySlot, properties {horizontalAlignment=HAlign_Center, verticalAlignment=VAlign_Center, padding=<ltrb 5 5 5 5>}}
  }
  Overlay_ClampArrows {
   widgetClassPath=&Overlay
   bIsVariable=true
   parent=Overlay_Main
   widget=Overlay_ClampArrows
   slot {class=OverlaySlot, properties {horizontalAlignment=HAlign_Fill, verticalAlignment=VAlign_Fill}}
  }
  Image_ClampLeft {
   widgetClassPath=&Image
   bIsVariable=true
   parent=Overlay_ClampArrows
   widget=Image_ClampLeft
   properties {
    brush {
     bIsDynamicallyLoaded=false
     drawAs=Image
     imageType=NoImage
     mirroring=NoMirror
     resourceName=None
     resourceObject=/Game/UI/CodeReferenced/T_UI_DirectionArrow
     tiling=NoTile
     imageSize=<32 32>
     margin=<ltrb 0 0 0 0>
     outlineSettings {
      bUseBrushTransparency=false
      roundingType=HalfHeightRadius
      width=0
      color {colorUseRule=UseColor_Specified, specifiedColor=<rgba 0 0 0 0>}
      cornerRadii=<0 0 0 0>
     }
     tintColor {colorUseRule=UseColor_Specified, specifiedColor=<rgba 1 1 1 1>}
     uVRegion {bIsValid=false, max=<0 0>, min=<0 0>}
    }
    renderTransform {angle=270, scale=<1 1>, shear=<0 0>, translation=<-32 0>}
   }
   slot {class=OverlaySlot, properties {verticalAlignment=VAlign_Center}}
  }
  Image_ClampTop {
   widgetClassPath=&Image
   bIsVariable=true
   parent=Overlay_ClampArrows
   widget=Image_ClampTop
   properties {
    brush {
     bIsDynamicallyLoaded=false
     drawAs=Image
     imageType=NoImage
     mirroring=NoMirror
     resourceName=None
     resourceObject=&T_UI_DirectionArrow
     tiling=NoTile
     imageSize=<32 32>
     margin=<ltrb 0 0 0 0>
     outlineSettings {
      bUseBrushTransparency=false
      roundingType=HalfHeightRadius
      width=0
      color {colorUseRule=UseColor_Specified, specifiedColor=<rgba 0 0 0 0>}
      cornerRadii=<0 0 0 0>
     }
     tintColor {colorUseRule=UseColor_Specified, specifiedColor=<rgba 1 1 1 1>}
     uVRegion {bIsValid=false, max=<0 0>, min=<0 0>}
    }
    renderTransform {angle=0, scale=<1 1>, shear=<0 0>, translation=<0 -32>}
   }
   slot {class=OverlaySlot, properties {horizontalAlignment=HAlign_Center}}
  }
  Image_ClampBottom {
   widgetClassPath=&Image
   bIsVariable=true
   parent=Overlay_ClampArrows
   widget=Image_ClampBottom
   properties {
    brush {
     bIsDynamicallyLoaded=false
     drawAs=Image
     imageType=NoImage
     mirroring=NoMirror
     resourceName=None
     resourceObject=&T_UI_DirectionArrow
     tiling=NoTile
     imageSize=<32 32>
     margin=<ltrb 0 0 0 0>
     outlineSettings {
      bUseBrushTransparency=false
      roundingType=HalfHeightRadius
      width=0
      color {colorUseRule=UseColor_Specified, specifiedColor=<rgba 0 0 0 0>}
      cornerRadii=<0 0 0 0>
     }
     tintColor {colorUseRule=UseColor_Specified, specifiedColor=<rgba 1 1 1 1>}
     uVRegion {bIsValid=false, max=<0 0>, min=<0 0>}
    }
    renderTransform {angle=0, scale=<1 -1>, shear=<0 0>, translation=<0 32>}
   }
   slot {class=OverlaySlot, properties {horizontalAlignment=HAlign_Center, verticalAlignment=VAlign_Bottom}}
  }
  Image_ClampRight {
   widgetClassPath=&Image
   bIsVariable=true
   parent=Overlay_ClampArrows
   widget=Image_ClampRight
   properties {
    brush {
     bIsDynamicallyLoaded=false
     drawAs=Image
     imageType=NoImage
     mirroring=NoMirror
     resourceName=None
     resourceObject=&T_UI_DirectionArrow
     tiling=NoTile
     imageSize=<32 32>
     margin=<ltrb 0 0 0 0>
     outlineSettings {
      bUseBrushTransparency=false
      roundingType=HalfHeightRadius
      width=0
      color {colorUseRule=UseColor_Specified, specifiedColor=<rgba 0 0 0 0>}
      cornerRadii=<0 0 0 0>
     }
     tintColor {colorUseRule=UseColor_Specified, specifiedColor=<rgba 1 1 1 1>}
     uVRegion {bIsValid=false, max=<0 0>, min=<0 0>}
    }
    renderTransform {angle=90, scale=<1 1>, shear=<0 0>, translation=<32 0>}
   }
   slot {class=OverlaySlot, properties {horizontalAlignment=HAlign_Right, verticalAlignment=VAlign_Center}}
  }
  TutorialHighlight {widgetClassPath=&TutorialHighlight}
 ]
}
